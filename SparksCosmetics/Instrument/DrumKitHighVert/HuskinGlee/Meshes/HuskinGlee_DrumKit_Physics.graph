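# Physics asset for the HuskinGlee drum kit cosmetic, setting up jointed motion for crash and ride pumpkin parts and a skull jaw.
# Capsule bodies form left and right chains under crash and ride pumpkin roots, plus a single jaw body under its own root.
# All three roots are angularly locked to the handle body, which also carries extra stem capsules.
# Chain swing limits loosen from base to tip, with twist locked on the inner joints and limited on the outer.
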PhysicsAsset /SparksCosmetics/Instrument/DrumKitHighVert/HuskinGlee/Meshes/HuskinGlee_DrumKit_Physics
bodies {
 crash_pumpkin_l_01 [{shapeType=Capsule, length=6.3163, radius=2.6171, shapeName=crash_pumpkin_l_01_capsule, center=<5.3 0.5 0>, rotation=<rot 90 0 0>}]
 crash_pumpkin_l_02 [{shapeType=Capsule, length=3.7505, radius=1.9373, shapeName=crash_pumpkin_l_02_capsule, center=<1.4 0 0>, rotation=<rot 90 0 0>}]
 crash_pumpkin_l_03 [{shapeType=Capsule, length=4.2781, radius=1.4098, shapeName=crash_pumpkin_l_03_capsule, center=<1.4 0 0>, rotation=<rot 90 0 0>}]
 crash_pumpkin_l_04 [{shapeType=Capsule, length=4.2781, radius=1.4098, shapeName=crash_pumpkin_l_04_capsule, center=<1.4 0 0>, rotation=<rot 90 0 0>}]
 crash_pumpkin_r_01 [{shapeType=Capsule, length=4.1965, radius=2.8464, shapeName=crash_pumpkin_r_01_capsule, center=<4 0 0>, rotation=<rot 90 0 0>}]
 crash_pumpkin_r_02 [{shapeType=Capsule, length=2.3704, radius=3.0923, shapeName=crash_pumpkin_r_02_capsule, center=<1.4 0 0>, rotation=<rot 90 0 0>}]
 crash_pumpkin_r_03 [{shapeType=Capsule, length=2.8072, radius=2.7645, shapeName=crash_pumpkin_r_03_capsule, center=<0.43 0 0>, rotation=<rot 90 0 0>}]
 crash_pumpkin_root [{shapeType=Capsule, length=1, radius=1, shapeName=crash_pumpkin_root_capsule}]
 handle [
  {
   shapeType=Capsule
   length=11.4415
   radius=1.7404
   shapeName=stem_l_01
   center=<69.2965 -34.751 105.0869>
   rotation=<rot -1.4144 -117.7041 -4.9104>
  }
  {shapeType=Capsule, length=1, radius=1, shapeName=handle_capsule}
  {
   shapeType=Capsule
   length=9.7898
   radius=1.972
   shapeName=stem_l_02
   center=<71.2941 -30.8316 93.6149>
   rotation=<rot -32.537 -124.6332 12.4623>
  }
  {
   shapeType=Capsule
   length=13.2154
   radius=2.471
   shapeName=stem_r_01
   center=<32.6594 89.7842 77.3938>
   rotation=<rot 0.0054 26.7218 -5.8739>
  }
  {
   shapeType=Capsule
   length=17.5826
   radius=2.471
   shapeName=handle_capsule
   center=<36.9139 84.4568 60.8514>
   rotation=<rot 8.5246 31.1976 45.4416>
  }
 ]
 ride_pumpkin_l_01 [{shapeType=Capsule, length=1.5191, radius=3.4268, shapeName=ride_pumpkin_l_01_capsule, center=<3 0 0>, rotation=<rot 90 0 0>}]
 ride_pumpkin_l_02 [{shapeType=Capsule, length=1.9567, radius=3.0615, shapeName=ride_pumpkin_l_02_capsule, center=<2.75 0 0>, rotation=<rot 90 0 0>}]
 ride_pumpkin_l_03 [{shapeType=Capsule, length=4.0311, radius=2.6499, shapeName=ride_pumpkin_l_03_capsule, center=<4 0 0>, rotation=<rot 90 0 0>}]
 ride_pumpkin_r_01 [{
   shapeType=Capsule
   length=6.1853
   radius=3.6015
   shapeName=ride_pumpkin_r_01_capsule
   center=<6 1 1.5>
   rotation=<rot 84.2511 -7.4166 -0.0001>
  }]
 ride_pumpkin_r_02 [{shapeType=Capsule, length=3.9812, radius=2.32, shapeName=ride_pumpkin_r_02_capsule, center=<2.27 0 0>, rotation=<rot 90 0 0>}]
 ride_pumpkin_r_03 [{shapeType=Capsule, length=4.4533, radius=1.3401, shapeName=ride_pumpkin_r_03_capsule, center=<2.27 0 0>, rotation=<rot 90 0 0>}]
 ride_pumpkin_r_04 [{shapeType=Capsule, length=4.4533, radius=1.3401, shapeName=ride_pumpkin_r_04_capsule, center=<2.27 0 0>, rotation=<rot 90 0 0>}]
 ride_pumpkin_root [{shapeType=Capsule, length=1, radius=1, shapeName=ride_pumpkin_root_capsule}]
 skull_jaw [{shapeType=Capsule, length=9.3515, radius=4.392, shapeName=skull_jaw_capsule, center=<8 0 0>, rotation=<rot 90 0 0>}]
 skull_jaw_root [{shapeType=Capsule, length=1, radius=1, shapeName=skull_jaw_root_capsule}]
}
constraints [
 {
  bone1Name=crash_pumpkin_l_01
  bone2Name=crash_pumpkin_root
  swing1LimitDegrees=10
  swing2LimitDegrees=10
  twistLimitDegrees=45
  twistMotion=Locked
 }
 {
  bone1Name=crash_pumpkin_l_02
  bone2Name=crash_pumpkin_l_01
  swing1LimitDegrees=15
  swing2LimitDegrees=15
  twistLimitDegrees=2
  twistMotion=Locked
 }
 {bone1Name=crash_pumpkin_l_03, bone2Name=crash_pumpkin_l_02, swing1LimitDegrees=45, swing2LimitDegrees=45, twistLimitDegrees=2}
 {bone1Name=crash_pumpkin_l_04, bone2Name=crash_pumpkin_l_03, swing1LimitDegrees=45, swing2LimitDegrees=45, twistLimitDegrees=2}
 {
  bone1Name=crash_pumpkin_r_01
  bone2Name=crash_pumpkin_root
  swing1LimitDegrees=10
  swing2LimitDegrees=10
  twistLimitDegrees=45
  twistMotion=Locked
 }
 {
  bone1Name=crash_pumpkin_r_02
  bone2Name=crash_pumpkin_r_01
  swing1LimitDegrees=15
  swing2LimitDegrees=15
  twistLimitDegrees=45
  twistMotion=Locked
 }
 {bone1Name=crash_pumpkin_r_03, bone2Name=crash_pumpkin_r_02, swing1LimitDegrees=45, swing2LimitDegrees=45, twistLimitDegrees=2}
 {bone1Name=skull_jaw, bone2Name=skull_jaw_root, swing1LimitDegrees=3, swing2LimitDegrees=10, twistLimitDegrees=2, twistMotion=Locked}
 {
  bone1Name=ride_pumpkin_l_01
  bone2Name=ride_pumpkin_root
  swing1LimitDegrees=10
  swing2LimitDegrees=10
  twistLimitDegrees=2
  twistMotion=Locked
 }
 {
  bone1Name=ride_pumpkin_r_01
  bone2Name=ride_pumpkin_root
  swing1LimitDegrees=10
  swing2LimitDegrees=10
  twistLimitDegrees=45
  twistMotion=Locked
 }
 {
  bone1Name=ride_pumpkin_l_02
  bone2Name=ride_pumpkin_l_01
  swing1LimitDegrees=15
  swing2LimitDegrees=15
  twistLimitDegrees=2
  twistMotion=Locked
 }
 {bone1Name=ride_pumpkin_l_03, bone2Name=ride_pumpkin_l_02, swing1LimitDegrees=45, swing2LimitDegrees=45, twistLimitDegrees=2}
 {
  bone1Name=ride_pumpkin_r_02
  bone2Name=ride_pumpkin_r_01
  swing1LimitDegrees=15
  swing2LimitDegrees=15
  twistLimitDegrees=2
  twistMotion=Locked
 }
 {bone1Name=ride_pumpkin_r_03, bone2Name=ride_pumpkin_r_02, swing1LimitDegrees=45, swing2LimitDegrees=45, twistLimitDegrees=2}
 {bone1Name=ride_pumpkin_r_04, bone2Name=ride_pumpkin_r_03, swing1LimitDegrees=45, swing2LimitDegrees=45, twistLimitDegrees=2}
 {
  bone1Name=crash_pumpkin_root
  bone2Name=handle
  swing1LimitDegrees=45
  swing1Motion=Locked
  swing2LimitDegrees=45
  swing2Motion=Locked
  twistLimitDegrees=45
  twistMotion=Locked
 }
 {
  bone1Name=skull_jaw_root
  bone2Name=handle
  swing1LimitDegrees=45
  swing1Motion=Locked
  swing2LimitDegrees=45
  swing2Motion=Locked
  twistLimitDegrees=45
  twistMotion=Locked
 }
 {
  bone1Name=ride_pumpkin_root
  bone2Name=handle
  swing1LimitDegrees=45
  swing1Motion=Locked
  swing2LimitDegrees=45
  swing2Motion=Locked
  twistLimitDegrees=45
  twistMotion=Locked
 }
]
