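# Animation Blueprint for the HuskinGlee drum kit, animating its crow and adding physics motion.
# A state machine loops Idle and, near the loop's end, plays Fidget A or B, chosen by a boolean flipped every animation update.
# The crow pose layers onto the input pose from crow_rig_root, then passes through RigidBody simulation relative to the handle bone.
# The random-player and additive nodes do not reach the output.
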AnimBlueprint /SparksCosmetics/Instrument/DrumKitHighVert/HuskinGlee/Meshes/HuskinGlee_DrumKit_AnimBP
target_skeleton=&/HuskinGlee_DrumKit_Skeleton
graphs [
 AnimGraph {nodes [
   n1 = Random Sequence Player!() -> (Pose:Pose Link Structure) {
    is=title AnimGraphNode_RandomPlayer
    entries [
     {
      chanceToPlay=0.7
      maxLoopCount=0
      maxPlayRate=1
      minLoopCount=0
      minPlayRate=1
      sequence=&/HuskinGlee_DrumKit_Crow_Idle_Loop
      blendIn {blendOption=Linear, blendTime=0.2, customCurve=null}
     }
     {
      chanceToPlay=0.3
      maxLoopCount=0
      maxPlayRate=1
      minLoopCount=0
      minPlayRate=1
      sequence=&/HuskinGlee_DrumKit_Crow_Fidget_A
      blendIn {blendOption=Linear, blendTime=0.2, customCurve=null}
     }
    ]
   }
   Apply Additive(Additive=@n1) {alphaBoolBlend {blendOption=Linear}}
   AnimGraphNode_StateMachine_0 = Crow State Machine() {is=title AnimGraphNode_StateMachine}
   n2 = Input Pose()
   n3 = RigidBody(ComponentPose=Local To Component(LocalPose=Layered blend per bone(BasePose=@n2, BlendPoses_0=@AnimGraphNode_StateMachine_0) {layerSetup [{branchFilters [{boneName=crow_rig_root}]}]})) {
    bOverrideWorldGravity=true
    overridePhysicsAsset=&/HuskinGlee_DrumKit_Physics
    simulationSpace=BaseBoneSpace
    alphaBoolBlend {blendOption=Linear}
    baseBoneRef {boneName=handle}
    componentLinearAccScale=<0.4 0.4 0.4>
    componentLinearVelScale=<0.2 0.2 0.2>
    overrideWorldGravity {z=-980}
   }
   Output Pose(Result=Component To Local(ComponentPose=@n3))
  ]}
 Crow State Machine {
  entry=Idle
  path=AnimGraph.AnimGraphNode_StateMachine_0.Crow State Machine
  ids=id
  nodes [
   AnimStateNode_0 = Idle()
   AnimStateNode_1 = Fidget_B() {is=title AnimStateNode}
   AnimStateNode_2 = Fidget_A() {is=title AnimStateNode}
   AnimStateTransitionNode_0 = AnimStateTransitionNode() {crossfadeDuration=0.4}
   AnimStateTransitionNode_1 = AnimStateTransitionNode() {bAutomaticRuleBasedOnSequencePlayerInState=true, crossfadeDuration=0.4}
   AnimStateTransitionNode_4 = AnimStateTransitionNode() {bAutomaticRuleBasedOnSequencePlayerInState=true, crossfadeDuration=0.4}
   AnimStateTransitionNode_5 = AnimStateTransitionNode() {crossfadeDuration=0.4}
  ]
  links [
   {
    dst=Idle
    rule=AnimGraph.AnimGraphNode_StateMachine_0.Crow State Machine.AnimStateTransitionNode_4.Transition
    src=Fidget_A
    via=AnimStateTransitionNode_4
   }
   {
    dst=Idle
    rule=AnimGraph.AnimGraphNode_StateMachine_0.Crow State Machine.AnimStateTransitionNode_1.Transition
    src=Fidget_B
    via=AnimStateTransitionNode_1
   }
   {
    dst=Fidget_A
    rule=AnimGraph.AnimGraphNode_StateMachine_0.Crow State Machine.AnimStateTransitionNode_5.Transition
    src=Idle
    via=AnimStateTransitionNode_5
   }
   {
    dst=Fidget_B
    rule=AnimGraph.AnimGraphNode_StateMachine_0.Crow State Machine.AnimStateTransitionNode_0.Transition
    src=Idle
    via=AnimStateTransitionNode_0
   }
  ]
 }
 Idle {
  path=AnimGraph.AnimGraphNode_StateMachine_0.Crow State Machine.AnimStateNode_0.Idle
  nodes [Output Animation Pose(Result=Sequence Player 'HuskinGlee_DrumKit_Crow_Idle_Loop'() {sequence=&HuskinGlee_DrumKit_Crow_Idle_Loop})]
 }
 Fidget_B {
  path=AnimGraph.AnimGraphNode_StateMachine_0.Crow State Machine.AnimStateNode_1.Fidget_B
  nodes [Output Animation Pose(Result=Sequence Player 'HuskinGlee_DrumKit_Crow_Fidget_B'() {bLoopAnimation=false, sequence=&/HuskinGlee_DrumKit_Crow_Fidget_B})]
 }
 Fidget_A {
  path=AnimGraph.AnimGraphNode_StateMachine_0.Crow State Machine.AnimStateNode_2.Fidget_A
  nodes [
   Random Sequence Player()
   Output Animation Pose(Result=Sequence Player 'HuskinGlee_DrumKit_Crow_Fidget_A'() {bLoopAnimation=false, sequence=&HuskinGlee_DrumKit_Crow_Fidget_A})
  ]
 }
 Transition {
  path=AnimGraph.AnimGraphNode_StateMachine_0.Crow State Machine.AnimStateTransitionNode_0.Transition
  nodes [
   n1 = Get isFidget_A() -> (isFidget_A:bool)
   n2 = "Get Relevant Anim Time Remaining (Idle)"!() -> (ReturnValue:float) {is=title K2Node_AnimGetter}
   Result(bCanEnterTransition=AND Boolean(A=NOT Boolean(A=@n1), B="float <= float"(A=@n2, B=0.2) -> (ReturnValue=~)))
  ]
 }
 Transition {path=AnimGraph.AnimGraphNode_StateMachine_0.Crow State Machine.AnimStateTransitionNode_1.Transition, nodes [Result()]}
 Transition {
  path=AnimGraph.AnimGraphNode_StateMachine_0.Crow State Machine.AnimStateTransitionNode_4.Transition
  nodes [Result(bCanEnterTransition=true)]
 }
 Transition {
  path=AnimGraph.AnimGraphNode_StateMachine_0.Crow State Machine.AnimStateTransitionNode_5.Transition
  nodes [
   n1 = "Get Relevant Anim Time Remaining (Idle)"()
   Result(bCanEnterTransition=AND Boolean(A=Get isFidget_A(), B="float <= float"(A=@n1, B=0.2) -> (ReturnValue=~)))
   "float <= float"(A="Get Relevant Anim Time Remaining Fraction (Idle)"!() -> (ReturnValue:float) {is=title K2Node_AnimGetter}, B=0.5) -> (ReturnValue=~)
  ]
 }
 EventGraph {nodes [
   Event BlueprintUpdateAnimation()
   > n1 = Flip Flop!(None:Exec) -> (A:Exec, B:Exec, IsA:bool=~) {is=title K2Node_MacroInstance}
   Set isFidget_A(execute=[@n1.A, @n1.B], isFidget_A:bool=@n1) -> (Output_Get:bool)
   TryGetPawnOwner()
  ]}
]
variables [isFidget_A {type=boolean}]
